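; Season 17 Emperor quest gameplay ability that applies the booth lock gameplay effect to the owning player until a quest objective is completed.
; On activation it calls the parent, then checks that the pawn matches the Emperor tag container and that the quest manager reports the first q05 objective incomplete.
; If both hold, it applies the effect to the pawn's ability system component.
; The ability ends immediately either way.
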(event Ability|EventActivateAbility
  (bind _owningplayerpawn (Variables|GAVars|Data|GetOwningPlayerPawn))
  (|CallParentFunction)
  (if (and (GameplayTags|TagContainer|CompareTagAssetInterfacetoOtherTags (Variables|GAVars|Setup|GetT_Emperor) _owningplayerpawn) (not (Quest|HasCompletedObjectivewithName (Quests|GetQuestManager (Fortnite|GetFortPlayerControllerfromActor _owningplayerpawn) "Invalid") "/BattlepassS17/Items/QuestItems/Emperor/Quest_S17_Emperor_q05.Quest_S17_Emperor_q05" "quest_s17_Emperor_q05_obj0" :!class /Script/FortniteGame.FortQuestManager)))
    (GameplayEffects|ApplyGameplayEffectToSelf :self (Ability|GetAbilitySystemComponent _owningplayerpawn) :GameplayEffectClass (Variables|GAVars|Setup|GetGE_BoothLock) :Level 0.0)
    (Ability|EndAbility)
    (else
      (Ability|EndAbility))))

(event Ability|EventFailedToActivatePassiveAbility
  (params (ActorInfo :type Gameplay%20Ability%20Actor%20Info%20Structure)))
